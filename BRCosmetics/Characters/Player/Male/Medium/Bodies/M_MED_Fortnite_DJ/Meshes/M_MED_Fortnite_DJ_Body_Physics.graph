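# Physics asset for the M_MED_Fortnite_DJ character body, providing secondary motion for hoodie strings and zippers.
# Capsules on the pelvis and upper spine form a torso volume, linked by free-rotating constraints.
# Two four-segment string chains hang from spine_05, and left and right zipper bodies attach to the pelvis.
# String and zipper joints lock twist and limit swing, so they dangle without spinning.
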PhysicsAsset /BRCosmetics/Characters/Player/Male/Medium/Bodies/M_MED_Fortnite_DJ/Meshes/M_MED_Fortnite_DJ_Body_Physics
bodies {
 dyn_hoodiestring_01_l [{shapeType=Capsule, length=4, radius=1, center=<2 0 0>, rotation=<rot 90 0 0>}]
 dyn_hoodiestring_01_r [{shapeType=Capsule, length=4, radius=1, center=<2 0 0>, rotation=<rot 90 0 0>}]
 dyn_hoodiestring_02_l [{shapeType=Capsule, length=4, radius=1, center=<2 0 0>, rotation=<rot 90 0 0>}]
 dyn_hoodiestring_02_r [{shapeType=Capsule, length=4, radius=1, center=<2 0 0>, rotation=<rot 90 0 0>}]
 dyn_hoodiestring_03_l [{shapeType=Capsule, length=4, radius=1, center=<2 0 0>, rotation=<rot 90 0 0>}]
 dyn_hoodiestring_03_r [{shapeType=Capsule, length=4, radius=1, center=<2 0 0>, rotation=<rot 90 0 0>}]
 dyn_hoodiestring_04_l [{shapeType=Capsule, length=4, radius=1, center=<2 0 0>, rotation=<rot 90 0 0>}]
 dyn_hoodiestring_04_r [{shapeType=Capsule, length=4, radius=1, center=<2 0 0>, rotation=<rot 90 0 0>}]
 dyn_zipper_l [{shapeType=Capsule, length=4, radius=1, center=<2.5 0 0>, rotation=<rot 90 0 0>}]
 dyn_zipper_r [{shapeType=Capsule, length=4, radius=1, center=<2.5 0 0>, rotation=<rot 90 0 0>}]
 pelvis [{shapeType=Capsule, length=11.3453, radius=12.0172}]
 spine_02 [{shapeType=Capsule, length=8.1859, radius=11.9222}]
 spine_03 [{shapeType=Capsule, length=5.3174, radius=12.9942, center=<5.0632 0.7086 0>}]
 spine_04 [{shapeType=Capsule, length=6.6681, radius=13.6326, center=<3.0419 0.1281 0>}]
 spine_05 [{shapeType=Capsule, length=11.556, radius=12.5789, center=<-0.6226 -0.0501 -0.0008>}]
}
constraints [
 {
  bone1Name=spine_05
  bone2Name=spine_04
  swing1LimitDegrees=45
  swing1Motion=Free
  swing2LimitDegrees=45
  swing2Motion=Free
  twistLimitDegrees=45
  twistMotion=Free
 }
 {
  bone1Name=spine_04
  bone2Name=spine_03
  swing1LimitDegrees=45
  swing1Motion=Free
  swing2LimitDegrees=45
  swing2Motion=Free
  twistLimitDegrees=45
  twistMotion=Free
 }
 {
  bone1Name=spine_03
  bone2Name=spine_02
  swing1LimitDegrees=45
  swing1Motion=Free
  swing2LimitDegrees=45
  swing2Motion=Free
  twistLimitDegrees=45
  twistMotion=Free
 }
 {bone1Name=dyn_hoodiestring_01_l, bone2Name=spine_05, swing1LimitDegrees=45, swing2LimitDegrees=45, twistLimitDegrees=45, twistMotion=Locked}
 {
  bone1Name=dyn_hoodiestring_02_l
  bone2Name=dyn_hoodiestring_01_l
  swing1LimitDegrees=45
  swing2LimitDegrees=45
  twistLimitDegrees=45
  twistMotion=Locked
 }
 {
  bone1Name=dyn_hoodiestring_03_l
  bone2Name=dyn_hoodiestring_02_l
  swing1LimitDegrees=45
  swing2LimitDegrees=45
  twistLimitDegrees=45
  twistMotion=Locked
 }
 {
  bone1Name=dyn_hoodiestring_04_l
  bone2Name=dyn_hoodiestring_03_l
  swing1LimitDegrees=45
  swing2LimitDegrees=45
  twistLimitDegrees=45
  twistMotion=Locked
 }
 {bone1Name=dyn_hoodiestring_01_r, bone2Name=spine_05, swing1LimitDegrees=45, swing2LimitDegrees=45, twistLimitDegrees=45, twistMotion=Locked}
 {
  bone1Name=dyn_hoodiestring_02_r
  bone2Name=dyn_hoodiestring_01_r
  swing1LimitDegrees=45
  swing2LimitDegrees=45
  twistLimitDegrees=45
  twistMotion=Locked
 }
 {
  bone1Name=dyn_hoodiestring_03_r
  bone2Name=dyn_hoodiestring_02_r
  swing1LimitDegrees=45
  swing2LimitDegrees=45
  twistLimitDegrees=45
  twistMotion=Locked
 }
 {
  bone1Name=dyn_hoodiestring_04_r
  bone2Name=dyn_hoodiestring_03_r
  swing1LimitDegrees=45
  swing2LimitDegrees=45
  twistLimitDegrees=45
  twistMotion=Locked
 }
 {
  bone1Name=spine_02
  bone2Name=pelvis
  swing1LimitDegrees=45
  swing1Motion=Free
  swing2LimitDegrees=45
  swing2Motion=Free
  twistLimitDegrees=45
  twistMotion=Free
 }
 {bone1Name=dyn_zipper_r, bone2Name=pelvis, swing1LimitDegrees=40, swing2LimitDegrees=40, twistLimitDegrees=45, twistMotion=Locked}
 {bone1Name=dyn_zipper_l, bone2Name=pelvis, swing1LimitDegrees=40, swing2LimitDegrees=40, twistLimitDegrees=45, twistMotion=Locked}
]
