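# Animation Blueprint for the body mesh of the M_MED_Fortnite_DJ player cosmetic, adding secondary motion to the hoodie strings and zipper.
# It copies the pose from the attached parent mesh, then constrains both string bones and the zipper bone to weighted blends of upper spine and left clavicle bones.
# A RigidBody node using the body physics asset follows, then an angle-limited AnimDynamics pass on the zipper, both with overridden gravity.
# The event graph holds no logic.
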AnimBlueprint /BRCosmetics/Characters/Player/Male/Medium/Bodies/M_MED_Fortnite_DJ/Meshes/M_MED_Fortnite_DJ_Body_AnimBP
target_skeleton=&/M_MED_Fortnite_DJ_Body_Skeleton
graphs [
 AnimGraph {nodes [
   n1 = "Constraint - dyn_hoodiestring_01_r (PPPP)"!(
    ComponentPose=Local To Component(LocalPose=Copy Pose From Mesh() {bUseAttachedParent=true})
    Alpha=1
    ConstraintWeights_0=0.1464
    ConstraintWeights_1:float=0.3501
    ConstraintWeights_2:float=0.1207
    ConstraintWeights_3:float=0.3828
   ) -> (Pose) {
    alphaBoolBlend {blendOption=Linear}
    boneToModify {boneName=dyn_hoodiestring_01_r}
    constraintSetup [{targetBone {boneName=spine_03}}, {targetBone {boneName=spine_04}}, {targetBone {boneName=spine_05}}, {targetBone {boneName=clavicle_l}}]
    constraintWeights [1,1,1,1]
   }
   n2 = "Constraint - dyn_hoodiestring_01_l (PPPP)"!(
    ComponentPose=@n1
    Alpha=1
    ConstraintWeights_0=0.1464
    ConstraintWeights_1:float=0.3501
    ConstraintWeights_2:float=0.1207
    ConstraintWeights_3:float=0.3828
   ) -> (Pose) {
    alphaBoolBlend {blendOption=Linear}
    boneToModify {boneName=dyn_hoodiestring_01_l}
    constraintSetup [{targetBone {boneName=spine_03}}, {targetBone {boneName=spine_04}}, {targetBone {boneName=spine_05}}, {targetBone {boneName=clavicle_l}}]
    constraintWeights [1,1,1,1]
   }
   n3 = "Constraint - dyn_hoodie_zipper (PPPP)"!(
    ComponentPose=@n2
    Alpha=1
    ConstraintWeights_0=0.0679
    ConstraintWeights_1:float=0.4676
    ConstraintWeights_2:float=0.3269
    ConstraintWeights_3:float=0.1376
   ) -> (Pose) {
    alphaBoolBlend {blendOption=Linear}
    boneToModify {boneName=dyn_hoodie_zipper}
    constraintSetup [{targetBone {boneName=spine_03}}, {targetBone {boneName=spine_04}}, {targetBone {boneName=spine_05}}, {targetBone {boneName=clavicle_l}}]
    constraintWeights [1,1,1,1]
   }
   n4 = RigidBody(ComponentPose=@n3) {
    bOverrideWorldGravity=true
    overridePhysicsAsset=&/M_MED_Fortnite_DJ_Body_Physics
    alphaBoolBlend {blendOption=Linear}
    componentLinearAccScale=<0.1 0.1 0.1>
    componentLinearVelScale=<0.1 0.1 0.1>
    overrideWorldGravity {z=-980}
   }
   n5 = Anim Dynamics - Bone: dyn_hoodie_zipper(ComponentPose=@n4, Alpha=1, ExternalForce="0,0,0") {
    angularDampingOverride=0.98
    bGravityOverrideInSimSpace=true
    bOverrideAngularDamping=true
    bUseGravityOverride=true
    alphaBoolBlend {blendOption=Linear}
    boundBone {boneName=dyn_hoodie_zipper}
    componentLinearAccScale=<0.2 0.2 0.2>
    componentLinearVelScale=<0.2 0.2 0.2>
    gravityOverride {z=-980}
    physicsBodyDefinitions [{
      sphereCollisionRadius=0
      boundBone {boneName=dyn_hoodie_zipper}
      boxExtents=<5 5 5>
      constraintSetup {angularLimitsMax=<0 30 30>, angularLimitsMin=<0 -5 -30>, angularTarget=<0 0 0>}
      localJointOffset=<5 0 0>
     }]
   }
   Output Pose(Result=Component To Local(ComponentPose=@n5)) {layerGroup=None}
  ]}
 EventGraph {nodes [
   TryGetPawnOwner()
   Event BlueprintUpdateAnimation()
  ]}
]
variables []
